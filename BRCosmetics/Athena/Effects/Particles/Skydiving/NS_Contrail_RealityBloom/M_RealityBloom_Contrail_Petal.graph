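# Translucent, unlit, two-sided petal material for the RealityBloom skydiving contrail, usable on Niagara mesh particles.
# Emissive tints two panning star-line textures with a vertical two-color gradient, warped by chromatic caustics that become constant on low-quality, forward and mobile paths.
# Opacity blends UV sphere masks, facing-based fade and gradients, with an optional overall fade.
# Panning noise masked along V builds a WPO term, while a per-particle random value offsets animation time.
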Material /BRCosmetics/Athena/Effects/Particles/Skydiving/NS_Contrail_RealityBloom/M_RealityBloom_Contrail_Petal
graph {
 properties {
  bUsedWithInstancedStaticMeshes=true
  bUsedWithNiagaraMeshParticles=true
  blendMode=BLEND_Translucent
  pixelDepthOffsetMode=PDOM_Legacy
  shadingModel=MSM_Unlit
  twoSided=true
  usedShadingModels=Unlit
 }
 nodes [
  n1 = ComponentMask(None=TextureCoordinate()) {g=true}
  n2 = ScalarParameter() {defaultValue=0.2, parameterName=ColorLerpOffset}
  n3 = Multiply(A=ComponentMask(None=NamedRerouteUsage()) {g=true, r=true}) {constB=0.06}
  n4 = Add(A=NamedRerouteUsage(), B=@n3)
  n5 = VectorParameter() {parameterName=ColorRibsUV_Pan, defaultValue=<rgba 0.1 0.7 0 0>}
  n6 = AppendVector(A=@n5.R, B=@n5.R)
  n7 = NamedRerouteUsage()
  n8 = Multiply(A=NamedRerouteUsage()) {constB=1.7}
  n9 = VectorParameter() {parameterName=ColorRibsUv2_Pan, defaultValue=<rgba -0.05 -0.5 0 1>}
  n10 = AppendVector(A=@n9.R, B=@n9.R)
  n11 = TextureSampleParameter2D(UVs=Panner(Coordinate=Subtract(A=@n8, B=@n3), Time=@n7, Speed=Multiply(A=@n6, B=@n10)) {speedX=1, speedY=1}) {parameterName=LinesTexture02, texture=/Game/Athena/Environments/Nexus/Textures/T_ZeroPoint_Line_Stars}
  n12 = VectorParameter() {parameterName=ColorRibs_Tint, defaultValue=<rgba 1 1 1 1>}
  n13 = ScalarParameter() {parameterName=ColorRibs_Intensity}
  n14 = Add(
   A=TextureSampleParameter2D(UVs=Panner(Coordinate=@n4, Time=@n7, Speed=@n6) {speedX=1, speedY=1}) {parameterName=LinesTexture01, texture=&T_ZeroPoint_Line_Stars}.RGB
   B=Add(A=Multiply(A=@n11.RGB, B=@n12.RGB) {constB=0.5}, B=Multiply(A=@n11.R, B=@n13))
  )
  n15 = NamedRerouteUsage()
  n16 = Add(A=Add(A=Multiply(A=@n14) {constB=0.8}, B=Multiply(A=@n15) {constB=0.5})) {constB=0.4}
  n17 = Multiply(
   A=LinearInterpolate(
    A=VectorParameter() {parameterName=Color_Top, defaultValue=<rgba 0.806 0.2066 1 1>}.RGB
    B=VectorParameter() {parameterName=Color_Bottom, defaultValue=<rgba 0.1476 0.5442 1 1>}.RGB
    Alpha=Add(A=@n1, B=@n2)
   )
   B=@n16
  )
  n18 = Multiply(A=Add(A=ScalarParameter() {parameterName=Intensity}), B=@n17)
  MaterialFunctionCall(Emissive=@n18) {materialFunction=&MF_FX_EmissiveAdjustments}
  Multiply(A=@n17, B=ParticleColor().RGB)
  n19 = NamedRerouteUsage()
  Multiply(A=@n16, B=LinearInterpolate(A=LinearInterpolate(Alpha=@n19) {constA=1, constB=0.5}, Alpha=VertexColor().R) {constAlpha=1, constB=3})
  NamedRerouteDeclaration(None=@n18) {name=Emissive, nodeColor=<rgba 1 0.1406 0 1>}
  n20 = ComponentMask(None=TextureCoordinate()) {g=true}
  n21 = LinearInterpolate(Alpha=Saturate(None=Multiply(A=@n20) {constB=1.25})) {constA=1, constB=0}
  n22 = Constant() {r=0.5}
  n23 = SphereMask(A=ComponentMask(None=TextureCoordinate() {coordinateIndex=1}) {r=true}, B=@n22, Radius=@n22) {hardnessPercent=0}
  n24 = MaterialFunctionCall(Float2=TextureCoordinate() {coordinateIndex=1}) {materialFunction=&BreakOutFloat2Components}
  n25 = Constant() {r=0.5}
  n26 = SphereMask(A=AppendVector(A=@n24.R, B=Power(Base=@n24.G)), B=@n25, Radius=@n25) {hardnessPercent=0}
  n27 = LinearInterpolate(Alpha=NamedRerouteDeclaration(None=Multiply(A=@n23, B=@n26)) {name=SphereMask UV, nodeColor=<rgba 0.1484 1 0 1>}) {constA=1, constB=0.4}
  n28 = LinearInterpolate(Alpha=Saturate(None=TwoSidedSign())) {constA=1, constB=0.6}
  n29 = TextureCoordinate()
  n30 = ComponentMask(None=@n29) {g=true}
  n31 = LinearInterpolate(Alpha=Saturate(None=Multiply(A=@n30) {constB=2.5})) {constA=0.1}
  n32 = Constant() {r=0.5}
  n33 = Power(Base=SphereMask(A=ComponentMask(None=@n29) {r=true}, B=@n32, Radius=@n32) {desc=Spline X opacity, hardnessPercent=0}) {constExponent=4}
  n34 = LinearInterpolate(
   A=Constant() {r=1}
   B=LinearInterpolate(A=Multiply(A=@n27, B=@n28), B=Multiply(A=@n31, B=@n33) {desc=Spline opacity}, Alpha=Constant() {r=0.7})
   Alpha=ScalarParameter() {defaultValue=1, parameterName=Translucency}
  )
  NamedRerouteDeclaration(None=StaticSwitchParameter(True=Multiply(A=@n21, B=@n34), False=@n34) {parameterName=UseOverallGradient?}) {name=Opacity, nodeColor=<rgba 1 0.4688 0 1>}
  n35 = OneMinus(None=ComponentMask(None=TextureCoordinate()) {g=true})
  n36 = ScalarParameter() {parameterName=WPO Bottom Mask Contrast}
  n37 = ScalarParameter() {defaultValue=4, group=WPO, parameterName=WPO Amnt Noise}
  n38 = TextureCoordinate() {coordinateIndex=1}
  n39 = AppendVector(
   A=Multiply(A=ComponentMask(None=@n38) {r=true}, B=ScalarParameter() {defaultValue=0.25, group=WPO, parameterName=WPO Tile U})
   B=Multiply(A=ComponentMask(None=@n38) {g=true}, B=ScalarParameter() {defaultValue=0.25, group=WPO, parameterName=WPO Tile V})
  )
  n40 = Time()
  n41 = AppendVector(
   A=ScalarParameter() {group=WPO, parameterName=WPO Pan Speed U}
   B=ScalarParameter() {defaultValue=0.2, group=WPO, parameterName=WPO Pan Speed V}
  )
  n42 = MaterialFunctionCall(
   In=TextureSample(UVs=Panner(Coordinate=@n39, Time=@n40, Speed=@n41) {speedX=-1, speedY=0.1}) {samplerType=SAMPLERTYPE_Grayscale, texture=&LowResBlurredNoise}.R
   Contrast=ScalarParameter() {parameterName=NoiseContrast}
  ) {materialFunction=&CheapContrast}
  NamedRerouteDeclaration(None=Multiply(
    A=MaterialFunctionCall(In=@n35, Contrast=@n36) {materialFunction=&CheapContrast}.Result
    B=Multiply(A=@n37, B=@n42.Result)
   )) {name=WPO, nodeColor=<rgba 0.7109 0 1 1>}
  n43 = Time()
  n44 = ScalarParameter() {defaultValue=0.1, parameterName=PanSpeed}
  NamedRerouteDeclaration(None=Add(A=DynamicParameter() {defaultValue=<rgba 0 1 1 0>, paramNames [Random, WPO Strength, NA, NA]}.Random, B=Multiply(A=@n43, B=@n44))) {name=Random UV Offset, nodeColor=<rgba 0.3359 0 1 1>}
  n45 = TextureCoordinate() {coordinateIndex=1}
  n46 = TextureCoordinate()
  n47 = VectorParameter() {parameterName=ColorRibsUV TIle, defaultValue=<rgba 0.5 0.5 0 0>}
  NamedRerouteDeclaration(None=Multiply(A=StaticSwitchParameter(True=@n45, False=@n46) {defaultValue=true, parameterName=UseCoord2ForTiling?}, B=AppendVector(A=@n47.R, B=@n47.R))) {name=ColorRibs_UV, nodeColor=<rgba 0.6641 1 0 1>}
  n48 = VectorParameter() {parameterName=ColorDistortion_Tile, defaultValue=<rgba 2 0.12 0 0>}
  n49 = Multiply(A=NamedRerouteUsage(), B=AppendVector(A=@n48.R, B=@n48.R))
  n50 = NamedRerouteUsage()
  n51 = VectorParameter() {parameterName=ColorDistortion_Pan, defaultValue=<rgba 0.1 0.25 0 0>}
  n52 = AppendVector(A=@n51.R, B=@n51.R)
  n53 = Constant() {r=0.25}
  n54 = ShadingPathSwitch(
   Default=TextureSample(UVs=Panner(Coordinate=@n49, Time=@n50, Speed=@n52) {speedX=1, speedY=1}) {texture=/Game/Effects/Fort_Effects/Textures/Tile/T_Caustics_Chromatic}.RGB
   Forward=@n53
   Mobile=@n53
  )
  n55 = Constant() {r=0.25}
  NamedRerouteDeclaration(None=QualitySwitch(Default=ShadingPathSwitch(Default=@n54, Forward=@n55, Mobile=@n55), Low=@n55)) {name=Distortion, nodeColor=<rgba 0 1 0.7188 1>}
  NamedRerouteUsage()
  NamedRerouteUsage()
  NamedRerouteUsage()
 ]
}
