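# Unlit, translucent, two-sided material for the Reality Bloom skydiving contrail, usable on spline meshes and Niagara ribbons or meshes.
# Panning Zero Point star-line textures and chromatic caustics, with distortion and randomized time, build the tinted glow over a background color blend.
# World position offset combines SplineThicken with masked, panning noise.
# Opacity combines texture-driven masks, UV edge and length fades, and particle alpha.
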Material /BRCosmetics/Athena/Effects/Particles/Skydiving/NS_Contrail_RealityBloom/M_RealityBloom_Contrail_Spline
graph {
 properties {
  bUsedWithInstancedStaticMeshes=true
  bUsedWithNiagaraMeshParticles=true
  bUsedWithNiagaraRibbons=true
  bUsedWithSplineMeshes=true
  blendMode=BLEND_Translucent
  pixelDepthOffsetMode=PDOM_Legacy
  shadingModel=MSM_Unlit
  twoSided=true
  usedShadingModels=Unlit
 }
 nodes [
  n1 = ScalarParameter() {parameterName=WidthBase}
  n2 = TextureCoordinate()
  n3 = MaterialFunctionCall(
   In=Add(A=ComponentMask(None=@n2) {g=true}, B=ScalarParameter() {defaultValue=0.25, parameterName=SplineThicknessBias})
   Contrast=ScalarParameter() {defaultValue=0.5, parameterName=SplineThicknessContrast}
  ) {materialFunction=&CheapContrast}
  n4 = ScalarParameter() {parameterName=WidthTip}
  n5 = ScalarParameter() {defaultValue=20, group=WPO, parameterName=WPO Amnt Noise}
  n6 = TextureCoordinate() {coordinateIndex=1}
  n7 = AppendVector(
   A=Multiply(A=ComponentMask(None=@n6) {r=true}, B=ScalarParameter() {group=WPO, parameterName=WPO Tile U})
   B=Multiply(A=ComponentMask(None=@n6) {g=true}, B=ScalarParameter() {defaultValue=0.25, group=WPO, parameterName=WPO Tile V})
  )
  n8 = Time()
  n9 = AppendVector(
   A=ScalarParameter() {group=WPO, parameterName=WPO Pan Speed U}
   B=ScalarParameter() {defaultValue=0.2, group=WPO, parameterName=WPO Pan Speed V}
  )
  n10 = MaterialFunctionCall(
   In=TextureSample(UVs=Panner(Coordinate=@n7, Time=@n8, Speed=@n9) {speedX=-1, speedY=0.1}) {samplerType=SAMPLERTYPE_Grayscale, texture=&LowResBlurredNoise}.R
   Contrast=ScalarParameter() {parameterName=NoiseContrast}
  ) {materialFunction=&CheapContrast}
  n11 = TextureCoordinate()
  n12 = Add(A=OneMinus(None=ComponentMask(None=@n11) {g=true}), B=ScalarParameter() {defaultValue=0.25, parameterName=WPO Bottom Mask Offset})
  n13 = ScalarParameter() {defaultValue=0.5, parameterName=WPO Bottom Mask Contrast}
  Add(
   A=MaterialFunctionCall(WidthBase=Add(A=@n1, B=@n3.Result), WidthTip=Add(A=@n4, B=@n3.Result)) {materialFunction=&SplineThicken}.WorldPositionOffset
   B=Multiply(
    A=Multiply(A=@n5, B=@n10.Result)
    B=MaterialFunctionCall(In=@n12, Contrast=@n13) {materialFunction=&CheapContrast}.Result
   )
  )
  n14 = NamedRerouteUsage()
  n15 = ParticleColor()
  MaterialFunctionCall(Emissive=Multiply(A=Multiply(A=@n14, B=@n15.RGB), B=ScalarParameter() {defaultValue=1, parameterName=EmissiveMult})) {materialFunction=&MF_FX_EmissiveAdjustments}
  n16 = MaterialFunctionCall(UV Channel=TextureCoordinate() {coordinateIndex=1}) {materialFunction=&LinearGradient}
  n17 = OneMinus(None=@n16.UGradient)
  n18 = TextureCoordinate() {coordinateIndex=1}
  n19 = Power(Base=OneMinus(None=ComponentMask(None=@n18) {g=true}), Exp=ScalarParameter() {defaultValue=2, parameterName=PowerBottomFade})
  n20 = Add(A=NamedRerouteUsage(), B=Add(A=NamedRerouteUsage(), B=NamedRerouteUsage()))
  n21 = NamedRerouteUsage()
  n22 = Constant() {r=0.25}
  n23 = Multiply(A=NamedRerouteDeclaration(None=MaterialFunctionCall(In=@n21, Contrast=@n22) {materialFunction=&CheapContrast}.Result) {name=Stuff4, nodeColor=<rgba 1 0.4922 0 1>}) {constB=0.4}
  n24 = Constant() {r=0.5}
  n25 = SphereMask(A=ComponentMask(None=TextureCoordinate() {coordinateIndex=1}) {r=true}, B=@n24, Radius=@n24) {hardnessPercent=0}
  n26 = Multiply(A=@n15.A, B=Multiply(A=Add(A=@n20, B=@n23) {constB=0.5}, B=NamedRerouteDeclaration(None=@n25) {name=Stuff3, nodeColor=<rgba 0 1 0.5078 1>}))
  Multiply(
   A=Multiply(A=Multiply(A=@n16.UGradient, B=@n17), B=Multiply(A=@n19, B=@n26))
   B=ScalarParameter() {defaultValue=1, parameterName=OpacityMult}
  )
  n27 = Time()
  n28 = ScalarParameter() {defaultValue=0.15, parameterName=PanSpeed}
  NamedRerouteDeclaration(None=Add(A=Multiply(A=@n27, B=@n28), B=DynamicParameter() {paramNames [Random,NA,Width,Zips]}.Random)) {name=Panner1_Time, nodeColor=<rgba 0.5938 0 1 1>}
  n29 = NamedRerouteUsage()
  n30 = VectorParameter() {parameterName=Tile2, defaultValue=<rgba 0.5 0.5 0 0>}
  n31 = AppendVector(A=@n30.R, B=@n30.R)
  n32 = VectorParameter() {parameterName=PanDirection2, defaultValue=<rgba 0.2 0.5 0 0>}
  n33 = TextureSampleParameter2D(UVs=Panner(Coordinate=Multiply(A=@n29, B=@n31), Time=NamedRerouteUsage(), Speed=AppendVector(A=@n32.R, B=@n32.R)) {speedX=1, speedY=1}) {parameterName=NoiseTexture, texture=/Game/Effects/Fort_Effects/Textures/Tile/T_Caustics_Chromatic}
  NamedRerouteDeclaration(None=Multiply(A=ComponentMask(None=@n33.RGB) {g=true, r=true}) {constB=0.05}) {name=Distortion, nodeColor=<rgba 1 0.9844 0 1>}
  NamedRerouteDeclaration(None=TextureCoordinate()) {name=SomeTileStuff, nodeColor=<rgba 1 0 0.6094 1>}
  n34 = NamedRerouteUsage()
  n35 = NamedRerouteUsage()
  n36 = TextureSampleParameter2D(UVs=Panner(Coordinate=Add(A=@n34, B=@n35), Time=NamedRerouteUsage(), Speed=NamedRerouteUsage()) {speedX=1, speedY=1}) {parameterName=LineTexture_01, texture=/Game/Athena/Environments/Nexus/Textures/T_ZeroPoint_Line_Stars}
  NamedRerouteDeclaration(None=@n36.B) {name=TextureSample_B, nodeColor=<rgba 0 1 0.5781 1>}
  n37 = Multiply(A=NamedRerouteUsage()) {constB=2}
  n38 = NamedRerouteUsage()
  n39 = VectorParameter() {parameterName=PanDirection1, defaultValue=<rgba 0.08 1 0 0>}
  n40 = NamedRerouteDeclaration(None=AppendVector(A=@n39.R, B=@n39.R)) {name=Panner1_Speed, nodeColor=<rgba 1 0 0.5625 1>}
  n41 = VectorParameter() {parameterName=PanDirection1Mult, defaultValue=<rgba -1.7 1.4 0 1>}
  n42 = AppendVector(A=@n41.R, B=@n41.R)
  n43 = TextureSampleParameter2D(UVs=Panner(Coordinate=Subtract(A=@n37, B=@n38), Time=NamedRerouteUsage(), Speed=Multiply(A=@n40, B=@n42)) {speedX=1, speedY=1}) {parameterName=LineTexture_02, texture=&T_ZeroPoint_Line_Stars}
  NamedRerouteDeclaration(None=@n43.B) {name=TextureSample2_B, nodeColor=<rgba 0.5703 0 1 1>}
  NamedRerouteDeclaration(None=@n33.B) {name=TextureSample3_B, nodeColor=<rgba 1 0.0234 0 1>}
  n44 = Add(
   A=@n36.RGB
   B=Multiply(A=@n43.RGB, B=VectorParameter() {parameterName=SecondaryTint, defaultValue=<rgba 0.8086 0.0573 1 1>}.RGB) {constB=0.5}
  )
  n45 = Multiply(A=VectorParameter() {parameterName=ZipGlow, defaultValue=<rgba 0.25 4.0005 8 0>}.RGB) {constB=5}
  n46 = Multiply(A=@n33.RGB, B=Add(A=Multiply(A=@n45) {constB=10}) {constB=0.5}) {constB=0.5}
  n47 = LinearInterpolate(
   A=VectorParameter() {parameterName=BackgroundColor1, defaultValue=<rgba 0.5608 0.062 1 0>}.RGB
   B=VectorParameter() {parameterName=BackgroundColor2, defaultValue=<rgba 0.0081 0.0162 0.1 0>}.RGB
  )
  n48 = NamedRerouteUsage()
  NamedRerouteDeclaration(None=Add(A=Add(A=@n44, B=@n46), B=Multiply(A=@n47, B=@n48))) {name=Emissive, nodeColor=<rgba 1 0 0.0703 1>}
  TextureCoordinate()
  TextureCoordinate()
 ]
}
