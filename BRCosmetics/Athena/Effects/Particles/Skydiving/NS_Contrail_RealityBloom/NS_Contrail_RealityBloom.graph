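# Niagara skydiving contrail for the Reality Bloom cosmetic, attached to two sampled bones or sockets on the player's skeletal mesh.
# Each side gets a trail mesh, layered petal meshes oriented by owner velocity, pulsing glow sprites with lights, plus flares and CPU/GPU glow dots.
# Left and right tendril ribbons follow Hermite splines between ring offsets, disturbed by curl noise.
# System.Dive scales spawn rates, noise and trail appearance, with front-end preview overrides.
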NiagaraSystem /BRCosmetics/Athena/Effects/Particles/Skydiving/NS_Contrail_RealityBloom/NS_Contrail_RealityBloom
dependencies {
 usedDataInterfaces [&NiagaraDataInterfaceArrayFloat, &NiagaraDataInterfaceCurve, &NiagaraDataInterfaceColorCurve]
 usedDynamicInputs [
  &SelectFloatFromArray
  /Niagara/DynamicInputs/Integer/V2/Round_Float
  &FN_FrontEndFloatModifier
  &SetBoolByFloatComparison
 ]
 usedModules [
  &self:SystemSpawnScript.SystemScriptSource.SystemScriptGraph.NiagaraNodeAssignment_1.SetVariables_#1
  /Game/Effects/Niagara/Modules/Skydiving/FN_DivingAttributes/v5/FN_DivingAttributes
  &self:SystemSpawnScript.SystemScriptSource.SystemScriptGraph.NiagaraNodeAssignment_0.SetVariables_#2
  &SystemState
  &FN_Integrator
 ]
}
emitters {
 Decal_AirGlow {
  emitterUpdateScript [
   &EmitterState()
   &SpawnBurst_Instantaneous(Spawn Count=2, Use Spawn Probability=false, UseLoopCountLimit=true)
  ]
  particleSpawnScript [
   &FN_SetSpawnGroup(Number of Spawn Groups=2)
   &InitializeParticle(
    Lifetime=&RandomRangeFloat(Minimum=4, Maximum=4.5)
    Sprite Rotation Mode=::NewEnumerator2
    Sprite Size Mode=::NewEnumerator3
    Sprite UV Mode=::NewEnumerator3
    Uniform Sprite Size=&RandomRangeFloat(Minimum=40, Maximum=50)
    UsePositionOffset=false
    Write Material Random=true
   )
  ]
  particleUpdateScript [
   &ParticleState(
    Kill Particles When Lifetime Has Elapsed=false
    Let Infinitely Lived Particles Die When Emitter Deactivates=true
    Loop Particles Lifetime=true
   )
   &SkeletalMeshLocation(
    Bone / Socket Index=&Add_Integer(A=$Particles.SpawnGroup, B=3)
    Bone Sampling Mode=::NewEnumerator4
    MaskSpawnGroup=false
    Sampled Position Offset=&MaskVectorBySpawnGroup(Masked Vector=<10 0 0>, Passthrough Vector=<-10 0 0>)
    Skeletal Mesh=$User.SkeletalMesh
   )
   &CameraOffset(Camera Offset Amount=10)
   &Color(Color=&LerpLinearColors(
     StartColor=<rgba 0 1.1962 2 1>
     EndColor=<rgba 1.5 0 0.2908 1>
     LerpFactor=&FloatFromCurve(FloatCurve=DataInterface() {propertyValues {Curve {keys [
          {interpMode=RCIM_Cubic, tangentMode=RCTM_User, time=0, value=0}
          {interpMode=RCIM_Cubic, tangentMode=RCTM_User, time=0.5, value=1}
          {tangentMode=RCTM_User, time=1, value=0}
         ]}}})
    ))
   &ScaleColor(
    Color Value To Scale=$Particles.Color
    Scale RGB=&VectorFromFloat(Value=0.5)
    ScaleA=true
    ScaleRGB=true
   )
   &ScaleSpriteSize(Scale Factor=&Vector2DFromFloat(Value=&FloatFromCurve(
      FloatCurve=DataInterface() {propertyValues {Curve {keys [
          {interpMode=RCIM_Cubic, tangentMode=RCTM_User, time=0, value=1}
          {interpMode=RCIM_Cubic, tangentMode=RCTM_User, time=0.5, value=1.1}
          {tangentMode=RCTM_User, time=1, value=1}
         ]}}}
      Scale Curve=0.8
     )))
   &FN_Sprite_Attributes(
    Color=&MultiplyLinearColor(Linear Color=$Particles.Color, Scale Factor=<1 1 1 1>)
    Create Color Attribute=true
    Create SpriteSize Attribute=true
    SpriteSize=&Multiply_Vector2D(A=$Particles.SpriteSize, B=<2 2>)
    Write to Module Namespace=true
   )
   &FN_Light_Attributes(
    Color=&LerpLinearColors(
     StartColor=<rgba 0.0269 0.25 0.974 1>
     EndColor=<rgba 1 0.1978 0.7007 1>
     LerpFactor=&FloatFromCurve(FloatCurve=DataInterface() {propertyValues {Curve {keys [
          {interpMode=RCIM_Cubic, tangentMode=RCTM_User, time=0, value=0}
          {interpMode=RCIM_Cubic, tangentMode=RCTM_User, time=0.5, value=1}
          {tangentMode=RCTM_User, time=1, value=0}
         ]}}})
    )
    Create Color Attribute=true
    Create Exponent Attribute=true
    Create Radius Attribute=true
    Exponent=1
    Radius=50
    Scale Color=&Lerp_Float(A=8, B=10, Alpha=$System.Dive)
   )
   &DynamicMaterialParameters(
    Index 0 Param 1=1.5
    Index 0 Param 3=5
    Index 0 Param 4=&RandomRangeFloat(Minimum=0, Maximum=1)
    Index 2 Param 1=3
    Index 2 Param 2=1
    Index 2 Param 3=1
    Index 2 Param 4=1
    Module.Param0WriteEnabled=true
    Module.Param0WriteEnabled2=true
    Module.Param1WriteEnabled=true
    Module.Param1WriteEnabled2=true
    Module.Param2WriteEnabled=true
    Module.Param2WriteEnabled2=true
    Module.Param3WriteEnabled=true
    Module.Param3WriteEnabled2=true
    Write Parameter Index 2=true
   )
  ]
  rendererClasses [&NiagaraSpriteRendererProperties, &NiagaraLightRendererProperties]
  renderers [
   {rendererClass=&NiagaraSpriteRendererProperties}
   {rendererClass=&NiagaraSpriteRendererProperties, rendererIndex=1}
   {rendererClass=&NiagaraLightRendererProperties, rendererIndex=2}
  ]
 }
 Flares {
  emitterUpdateScript [
   &EmitterState()
   &SpawnRate(SpawnRate=&Lerp_Float(A=32, B=60, Alpha=$System.Dive), Use Spawn Probability=false)
  ]
  particleSpawnScript [
   &FN_SetSpawnGroup(Number of Spawn Groups=2)
   &InitializeParticle(
    Color Maximum=<rgba 0.2316 1.2823 2 1>
    Color Minimum=<rgba 1.722 0.1004 2 1>
    Color Mode=::NewEnumerator2
    Lifetime Max=1
    Lifetime Min=0.5
    Lifetime Mode=::NewEnumerator1
    Sprite Size Mode=::NewEnumerator1
    Uniform Sprite Size Max=20
    UsePositionOffset=false
    Write Material Random=true
   )
   &SkeletalMeshLocation(
    Bone / Socket Index=&Add_Integer(A=$Particles.SpawnGroup, B=3)
    Bone Sampling Mode=::NewEnumerator4
    Interpolate Mesh Sampling=false
    MaskSpawnGroup=false
    Sampled Position Offset=&MaskVectorBySpawnGroup(Masked Vector=<10 0 0>, Passthrough Vector=<-10 0 0>)
    Skeletal Mesh=$User.SkeletalMesh
   )
   &ShapeLocation(MaskSpawnGroup=false, Override Seed=false, Sphere Radius=40, UseOwnerScale=true)
   &AddVelocity(
    Module.Use Falloff Distance=false
    Override Seed=false
    Velocity Mode=::NewEnumerator1
    Velocity Origin:Vector3f=$Particles.SkeletalMeshLocation.SampledPosition
    Velocity Speed=&FloatFromCurve(
     FloatCurve=DataInterface() {propertyValues {Curve {keys [<at 0.25 5 RCIM_Cubic>, {interpMode=RCIM_Cubic, leaveTangent=-228.4063, time=0.75, value=100}]}}}
     CurveIndex=&RandomRangeFloat(Minimum=0, Maximum=1)
     Scale Curve=-1
    )
   )
  ]
  particleUpdateScript [
   &ParticleState()
   &InheritVelocity(Inherited Velocity Amount Scale=<0.5 0.3 -0.05>, Inherited Velocity Speed Limit=0.5, Speed Limit=true, UseSpeedThreshold=false)
   &AccelerationForce(
    Acceleration=&Lerp_Vector(
     A=<0 0 150>
     B=<0 0 980>
     Alpha=&FloatFromCurve(FloatCurve=DataInterface() {propertyValues {Curve {keys [
          {arriveTangent=15.6865, interpMode=RCIM_Cubic, leaveTangent=15.6865, tangentMode=RCTM_User, time=0, value=0}
          <at 0.1408 0.9847 RCIM_Cubic>
          {arriveTangent=-0.0405, interpMode=RCIM_Cubic, leaveTangent=-0.0405, tangentMode=RCTM_User, time=0.5, value=0}
         ]}}})
    )
    Coordinate Space=::Local
   )
   /Niagara/Modules/Update/Forces/CurlNoiseForce(
    Noise Frequency=1
    Noise Quality / Cost=::NewEnumerator0
    Noise Strength=&FloatFromCurve(
     FloatCurve=DataInterface() {propertyValues {Curve {keys [{interpMode=RCIM_Cubic, tangentMode=RCTM_User, time=0, value=1}, {interpMode=RCIM_Cubic, tangentMode=RCTM_User, time=1, value=-1}]}}}
     Scale Curve=150
    )
    Pan Noise=false
   )
   &Drag(Drag=&RandomRangeFloat(Minimum=0.5, Maximum=1), Use Linear Drag=true, Use Rotational Drag=false)
   &RotateAroundPoint(
    Radius:NiagaraFloat=&FloatFromCurve(
     FloatCurve=DataInterface() {propertyValues {Curve {keys [{interpMode=RCIM_Cubic, tangentMode=RCTM_User, time=0, value=0.25}, {interpMode=RCIM_Cubic, tangentMode=RCTM_User, time=1, value=1}]}}}
     Scale Curve=6
    )
    Rotation Rate:NiagaraFloat=&RandomRangeFloat(Minimum=0.5, Maximum=1)
    Set Particle Position=false
   )
   &SolveForcesAndVelocity(Clamp Velocity=false, Limit Acceleration=false)
   /SetVariables(Particles.Rotational_Position:NiagaraPosition=&Add_Vector(
     A=Linked() {linkedVariable {name=Particles.Position, type=NiagaraPosition}}
     B=Linked() {linkedVariable {name=Output.RotateAroundPoint.OutputPosition, type=NiagaraPosition}}
    )) {isSetParameters=true, module=SetVariables_#3}
   &ScaleColor(
    Scale Alpha=&FloatFromCurve(
     FloatCurve=DataInterface() {propertyValues {Curve {keys [
         {arriveTangent=26.7273, interpMode=RCIM_Cubic, leaveTangent=26.7273, tangentMode=RCTM_User, time=0, value=0}
         {interpMode=RCIM_Cubic, tangentMode=RCTM_User, time=0.0967, value=0.9767}
         {interpMode=RCIM_Cubic, tangentMode=RCTM_User, time=0.66, value=1}
         <at 1 0 RCIM_Cubic>
        ]}}}
     Scale Curve=&RandomRangeFloat(Minimum=0.5, Maximum=2)
    )
    ScaleA=true
    ScaleRGB=true
   )
   &ScaleSpriteSize(Scale Factor=&Vector2DFromFloat(Value=&FloatFromCurve(FloatCurve=DataInterface() {propertyValues {Curve {keys [<at -0.0046 0>,<at 0.025 1>,<at 0.66 1>,<at 1.0011 0.002>]}}})))
  ]
  rendererClasses [&NiagaraSpriteRendererProperties]
  renderers [{rendererClass=&NiagaraSpriteRendererProperties}]
 }
 Glow_Dots_CPU {
  emitterUpdateScript [
   &EmitterState()
   &SpawnRate(SpawnRate=&Lerp_Float(A=30, B=55, Alpha=$System.Dive), Use Spawn Probability=false)
  ]
  particleSpawnScript [
   &FN_SetSpawnGroup(Number of Spawn Groups=2)
   &InitializeParticle(
    Lifetime Max=1
    Lifetime Min=0.5
    Lifetime Mode=::NewEnumerator1
    Sprite Size Mode=::NewEnumerator1
    Uniform Sprite Size Max=8
    Uniform Sprite Size Min=4.5
    UsePositionOffset=false
    Write Material Random=true
   )
   &SkeletalMeshLocation(
    Bone / Socket Index=&Add_Integer(A=$Particles.SpawnGroup, B=3)
    Bone Sampling Mode=::NewEnumerator4
    Interpolate Mesh Sampling=false
    MaskSpawnGroup=false
    Sampled Position Offset=&MaskVectorBySpawnGroup(Masked Vector=<10 0 0>, Passthrough Vector=<-10 0 0>)
    Skeletal Mesh=$User.SkeletalMesh
   )
   &ShapeLocation(MaskSpawnGroup=false, Override Seed=false, Sphere Radius=20, UseOwnerScale=true)
   &AddVelocity(
    Module.Use Falloff Distance=false
    Override Seed=false
    Velocity Mode=::NewEnumerator1
    Velocity Origin:Vector3f=$Particles.SkeletalMeshLocation.SampledPosition
    Velocity Speed=&FloatFromCurve(
     FloatCurve=DataInterface() {propertyValues {Curve {keys [<at 0.25 5 RCIM_Cubic>, {interpMode=RCIM_Cubic, leaveTangent=-228.4063, time=0.75, value=100}]}}}
     CurveIndex=&RandomRangeFloat(Minimum=0, Maximum=1)
     Scale Curve=-1
    )
   )
   /SetVariables(Particles.VisibilityTag=&RandomRangeInt(Minimum=0, Maximum=1)) {isSetParameters=true, module=SetVariables_#4}
  ]
  particleUpdateScript [
   &ParticleState()
   &InheritVelocity(Inherited Velocity Amount Scale=<0.5 0.3 -0.05>, Inherited Velocity Speed Limit=0, Speed Limit=true, UseSpeedThreshold=false)
   &AccelerationForce(
    Acceleration=&Lerp_Vector(
     A=<0 0 150>
     B=<0 0 980>
     Alpha=&FloatFromCurve(FloatCurve=DataInterface() {propertyValues {Curve {keys [
          {arriveTangent=15.6865, interpMode=RCIM_Cubic, leaveTangent=15.6865, tangentMode=RCTM_User, time=0, value=0}
          <at 0.1408 0.9847 RCIM_Cubic>
          {arriveTangent=-0.0405, interpMode=RCIM_Cubic, leaveTangent=-0.0405, tangentMode=RCTM_User, time=0.5, value=0}
         ]}}})
    )
    Coordinate Space=::Local
   )
   &CurlNoiseForce(
    Noise Frequency=1
    Noise Quality / Cost=::NewEnumerator0
    Noise Strength=&FloatFromCurve(
     FloatCurve=DataInterface() {propertyValues {Curve {keys [{interpMode=RCIM_Cubic, tangentMode=RCTM_User, time=0, value=1}, {interpMode=RCIM_Cubic, tangentMode=RCTM_User, time=1, value=-1}]}}}
     Scale Curve=150
    )
    Pan Noise=false
   )
   &Drag(Drag=&RandomRangeFloat(Minimum=0.5, Maximum=1), Use Linear Drag=true, Use Rotational Drag=false)
   &RotateAroundPoint(
    Radius:NiagaraFloat=&FloatFromCurve(
     FloatCurve=DataInterface() {propertyValues {Curve {keys [{interpMode=RCIM_Cubic, tangentMode=RCTM_User, time=0, value=0.25}, {interpMode=RCIM_Cubic, tangentMode=RCTM_User, time=1, value=1}]}}}
     Scale Curve=6
    )
    Rotation Rate:NiagaraFloat=&RandomRangeFloat(Minimum=0.5, Maximum=1)
    Set Particle Position=false
   )
   &SolveForcesAndVelocity(Clamp Velocity=false, Limit Acceleration=false)
   /SetVariables(Particles.Rotational_Position:NiagaraPosition=&Add_Vector(
     A=Linked() {linkedVariable {name=Particles.Position, type=NiagaraPosition}}
     B=Linked() {linkedVariable {name=Output.RotateAroundPoint.OutputPosition, type=NiagaraPosition}}
    )) {isSetParameters=true, module=SetVariables_#3}
   &ScaleColor(
    Scale Alpha=&FloatFromCurve(
     FloatCurve=DataInterface() {propertyValues {Curve {keys [
         {arriveTangent=0.0704, interpMode=RCIM_Cubic, leaveTangent=0.0704, tangentMode=RCTM_User, time=0, value=0}
         {interpMode=RCIM_Cubic, tangentMode=RCTM_User, time=0.15, value=0.9767}
         {interpMode=RCIM_Cubic, tangentMode=RCTM_User, time=0.66, value=1}
         <at 1 0 RCIM_Cubic>
        ]}}}
     Scale Curve=&RandomRangeFloat(Minimum=0.5, Maximum=2)
    )
    ScaleA=true
    ScaleRGB=true
   )
   &ScaleSpriteSize(Scale Factor=&Vector2DFromFloat(Value=&FloatFromCurve(FloatCurve=DataInterface() {propertyValues {Curve {keys [<at -0.0046 0>,<at 0.025 1>,<at 0.66 1>,<at 1.0011 0.002>]}}})))
  ]
  rendererClasses [&NiagaraSpriteRendererProperties]
  renderers [
   {rendererClass=&NiagaraSpriteRendererProperties}
   {rendererClass=&NiagaraSpriteRendererProperties, rendererIndex=1}
  ]
 }
 Glow_Dots_GPU {
  simTarget=GPUComputeSim
  emitterUpdateScript [
   &EmitterState()
   &SpawnRate(SpawnRate=&Lerp_Float(A=75, B=100, Alpha=$System.Dive), Use Spawn Probability=false)
  ]
  particleSpawnScript [
   &FN_SetSpawnGroup(Number of Spawn Groups=2)
   &InitializeParticle(
    Lifetime Max=1
    Lifetime Min=0.5
    Lifetime Mode=::NewEnumerator1
    Sprite Size Mode=::NewEnumerator1
    Uniform Sprite Size Max=8
    Uniform Sprite Size Min=4.5
    UsePositionOffset=false
    Write Material Random=true
   )
   &SkeletalMeshLocation(
    Bone / Socket Index=&Add_Integer(A=$Particles.SpawnGroup, B=3)
    Bone Sampling Mode=::NewEnumerator4
    Interpolate Mesh Sampling=false
    MaskSpawnGroup=false
    Sampled Position Offset=&MaskVectorBySpawnGroup(Masked Vector=<10 0 0>, Passthrough Vector=<-10 0 0>)
    Skeletal Mesh=$User.SkeletalMesh
   )
   &ShapeLocation(MaskSpawnGroup=false, Override Seed=false, Sphere Radius=20, UseOwnerScale=true)
   &AddVelocity(
    Module.Use Falloff Distance=false
    Override Seed=false
    Velocity Mode=::NewEnumerator1
    Velocity Origin:Vector3f=$Particles.SkeletalMeshLocation.SampledPosition
    Velocity Speed=&FloatFromCurve(
     FloatCurve=DataInterface() {propertyValues {Curve {keys [<at 0.25 5 RCIM_Cubic>, {interpMode=RCIM_Cubic, leaveTangent=-228.4063, time=0.75, value=100}]}}}
     CurveIndex=&RandomRangeFloat(Minimum=0, Maximum=1)
     Scale Curve=-1
    )
   )
   /SetVariables(Particles.VisibilityTag=&RandomRangeInt(Minimum=0, Maximum=1)) {isSetParameters=true, module=SetVariables_#4}
  ]
  particleUpdateScript [
   &ParticleState()
   &InheritVelocity(Inherited Velocity Amount Scale=<0.5 0.3 -0.05>, Inherited Velocity Speed Limit=0, Speed Limit=true, UseSpeedThreshold=false)
   &AccelerationForce(
    Acceleration=&Lerp_Vector(
     A=<0 0 150>
     B=<0 0 980>
     Alpha=&FloatFromCurve(FloatCurve=DataInterface() {propertyValues {Curve {keys [
          {arriveTangent=15.6865, interpMode=RCIM_Cubic, leaveTangent=15.6865, tangentMode=RCTM_User, time=0, value=0}
          <at 0.1408 0.9847 RCIM_Cubic>
          {arriveTangent=-0.0405, interpMode=RCIM_Cubic, leaveTangent=-0.0405, tangentMode=RCTM_User, time=0.5, value=0}
         ]}}})
    )
    Coordinate Space=::Local
   )
   &CurlNoiseForce(
    Noise Frequency=1
    Noise Quality / Cost=::NewEnumerator0
    Noise Strength=&FloatFromCurve(
     FloatCurve=DataInterface() {propertyValues {Curve {keys [{interpMode=RCIM_Cubic, tangentMode=RCTM_User, time=0, value=1}, {interpMode=RCIM_Cubic, tangentMode=RCTM_User, time=1, value=-1}]}}}
     Scale Curve=150
    )
    Pan Noise=false
   )
   &Drag(Drag=&RandomRangeFloat(Minimum=0.5, Maximum=1), Use Linear Drag=true, Use Rotational Drag=false)
   &RotateAroundPoint(
    Radius:NiagaraFloat=&FloatFromCurve(
     FloatCurve=DataInterface() {propertyValues {Curve {keys [{interpMode=RCIM_Cubic, tangentMode=RCTM_User, time=0, value=0.25}, {interpMode=RCIM_Cubic, tangentMode=RCTM_User, time=1, value=1}]}}}
     Scale Curve=6
    )
    Rotation Rate:NiagaraFloat=&RandomRangeFloat(Minimum=0.5, Maximum=1)
    Set Particle Position=false
   )
   &SolveForcesAndVelocity(Clamp Velocity=false, Limit Acceleration=false)
   /SetVariables(Particles.Rotational_Position:NiagaraPosition=&Add_Vector(
     A=Linked() {linkedVariable {name=Particles.Position, type=NiagaraPosition}}
     B=Linked() {linkedVariable {name=Output.RotateAroundPoint.OutputPosition, type=NiagaraPosition}}
    )) {isSetParameters=true, module=SetVariables_#3}
   &ScaleColor(
    Scale Alpha=&FloatFromCurve(
     FloatCurve=DataInterface() {propertyValues {Curve {keys [
         {arriveTangent=0.0704, interpMode=RCIM_Cubic, leaveTangent=0.0704, tangentMode=RCTM_User, time=0, value=0}
         {interpMode=RCIM_Cubic, tangentMode=RCTM_User, time=0.15, value=0.9767}
         {interpMode=RCIM_Cubic, tangentMode=RCTM_User, time=0.66, value=1}
         <at 1 0 RCIM_Cubic>
        ]}}}
     Scale Curve=&RandomRangeFloat(Minimum=0.5, Maximum=2)
    )
    ScaleA=true
    ScaleRGB=true
   )
   &ScaleSpriteSize(Scale Factor=&Vector2DFromFloat(Value=&FloatFromCurve(FloatCurve=DataInterface() {propertyValues {Curve {keys [<at -0.0046 0>,<at 0.025 1>,<at 0.66 1>,<at 1.0011 0.002>]}}})))
  ]
  rendererClasses [&NiagaraSpriteRendererProperties]
  renderers [
   {rendererClass=&NiagaraSpriteRendererProperties}
   {rendererClass=&NiagaraSpriteRendererProperties, rendererIndex=1}
  ]
 }
 Petals_Mesh {
  emitterUpdateScript [
   &EmitterState()
   &SpawnBurst_Instantaneous(Spawn Count=32, Use Spawn Probability=false, UseLoopCountLimit=true)
  ]
  particleSpawnScript [
   &FN_SetSpawnGroup(Inclusivity=true, Normalize Spawn Group=true, Particles Per Group=8)
   /SetVariables(
    Particles.SpawnEven:NiagaraFloat=&MaskFloatBySpawnGroup(
     Masked Float=0
     Passthrough Float=&MaskFloatBySpawnGroup(Masked Float=0, Passthrough Float=1, Spawn Group Mask=2)
    )
    Particles.SpawnGroupFloat:NiagaraFloat=&MakeCustomFloatFromBool(
     Bool=&SetBoolByIntComparison(A=$Particles.SpawnGroup, B=2, Comparison Type=::NewEnumerator2)
     True Float=1
     False Float=0
    )
   ) {isSetParameters=true}
   /SetVariables(Particles.VisibilityTag=&MakeIntFromFloat(Float=$Particles.SpawnEven)) {isSetParameters=true}
   &InitializeParticle(
    Color=&LerpLinearColors(StartColor=<rgba 2 2 2 2>, EndColor=<rgba 1 1 1 0.25>, LerpFactor=$Particles.SpawnEven)
    Mesh Scale=&Lerp_Vector(A=<1.75 1.75 2>, B=<1.9 1.9 2.5>, Alpha=$Particles.SpawnEven)
    Mesh Scale Mode=::NewEnumerator4
    UsePositionOffset=false
    Write Material Random=true
   )
   &InitialMeshOrientation(
    Mesh Axis To Orient=&FN_FrontEndVector3Modifier(Gameplay=<0 0 1>, Front End=<0 0 -1>)
    Mesh Orientation Mode=::NewEnumerator2
    Rotation=&MakeVector(X=0, Y=0, Z=$Particles.Normalized Spawn Group)
    Use Rotation Vector=true
   )
   &CameraOffset(Camera Offset Amount=-5)
   &DynamicMaterialParameters(
    Index 0 Param 1=&RandomRangeFloat(Minimum=0, Maximum=1)
    Module.Param0WriteEnabled=true
    Module.Param1WriteEnabled=true
    Module.Param2WriteEnabled=true
    Module.Param3WriteEnabled=true
   )
  ]
  particleUpdateScript [
   &ParticleState(Kill Particles When Lifetime Has Elapsed=false)
   &SkeletalMeshLocation(
    Bone / Socket Index=&Add_Integer(A=/Niagara/DynamicInputs/Divide/Divide_Int(A:NiagaraInt32=$Particles.SpawnGroup, B:NiagaraInt32=2), B=3)
    Bone Sampling Mode=::NewEnumerator4
    MaskSpawnGroup=false
    Orientation Sampling=::NewEnumerator1
    Sampled Position Offset=&Multiply_Vector(
     A=&Lerp_Vector(A=<2 0 -16>, B=<2 0 -12>, Alpha=$Particles.SpawnEven)
     B=&MakeVector(
      X=1
      Y=1
      Z=&FloatFromCurve(
       FloatCurve=DataInterface() {propertyValues {Curve {keys [<at 0 0.85>,<at 0.125 1>,<at 0.25 0.85>,<at 0.375 1>,<at 0.5 0.85>,<at 0.625 1>,<at 0.75 0.85>,<at 0.875 1>,<at 1 0.85>]}}}
       CurveIndex=$Particles.Normalized Spawn Group
      )
     )
    )
    Sampled Position Offset Coordinate Space:ENiagaraSkelMeshTransforms=::NewEnumerator0
    Skeletal Mesh=$User.SkeletalMesh
   )
   /SetVariables(Particles.MeshOrientation=&OrientMeshToVector(
     Look At Direction=&NormalizeVector(
      Vector To Normalize=&TransformVector(
       Vector=&Multiply_Vector(A=$Engine.Owner.Velocity, B=<0.2 0.2 1>)
       Source Space=::World
       Destination Space=::Local
      )
      Safe Normalize=true
      Safe Fallback Threshold=0
      Safe Fallback Vector=<0 0 0.001>
     )
     Reference Axis Vector=<-1 0 0>
    )) {isSetParameters=true}
  ]
  rendererClasses [&NiagaraMeshRendererProperties]
  renderers [
   {rendererClass=&NiagaraMeshRendererProperties, rendererIndex=0}
   {rendererClass=&NiagaraMeshRendererProperties, rendererIndex=1}
  ]
 }
 Tendril_Ribbon_L {
  emitterUpdateScript [
   &EmitterState()
   &SpawnRate(SpawnRate=60, Use Spawn Probability=false)
  ]
  particleSpawnScript [
   &FN_SetSpawnGroup(Number of Spawn Groups=8)
   &InitializeParticle(
    Color=<rgba 1 1 1 0.2>
    Ribbon Width=1.5
    Ribbon Width Mode=::NewEnumerator1
    Sprite Size Mode=::NewEnumerator3
    Uniform Sprite Size=4
    UsePositionOffset=false
    Write Material Random=true
   )
   /SetVariables(Particles.FloatSpawnGroup:NiagaraFloat=&Divide_Float(A=&MakeFloatFromInt(Int=$Particles.SpawnGroup), B=8)) {isSetParameters=true, module=SetVariables_#5}
   &ShapeLocation(
    Apply To Particle Position=::NewEnumerator1
    MaskSpawnGroup=false
    Offset=<0 0 -15>
    Offset Coordinate Space=::World
    Offset Mode=::NewEnumerator0
    Override Seed=false
    Ring / Disc Distribution Mode:ENiagaraSphereDistributionMode=::NewEnumerator1
    Ring Radius=8
    Rotation Coordinate Space=::Simulation
    Shape Primitive=::NewEnumerator4
    U Position=$Particles.FloatSpawnGroup
    UseOwnerScale=true
   )
   &ShapeLocation(
    Apply To Particle Position=::NewEnumerator1
    MaskSpawnGroup=false
    Offset=&RandomRangeVector(Minimum=<0 0 50>, Maximum=<0 0 120>, Randomness Mode=::NewEnumerator1, Random Seed=$Particles.SpawnGroup, Fixed Random Seed=true)
    Offset Coordinate Space=::World
    Offset Mode=::NewEnumerator0
    Override Seed=false
    Ring / Disc Distribution Mode:ENiagaraSphereDistributionMode=::NewEnumerator1
    Ring Radius=8
    Rotation Coordinate Space=::Simulation
    Shape Primitive=::NewEnumerator4
    U Position=$Particles.FloatSpawnGroup
    UseOwnerScale=true
   ) {module=ShapeLocation001}
  ]
  particleUpdateScript [
   &ParticleState()
   &InheritVelocity(Inherited Velocity Amount Scale=<0.5 0.1 -0.5>, Inherited Velocity Speed Limit=2, Speed Limit=true, UseSpeedThreshold=false)
   /Niagara/Modules/Update/Forces/V2/CurlNoiseForce(
    Noise Frequency=&Lerp_Float(A=20, B=50, Alpha=$System.Dive)
    Noise Strength=&Lerp_Float(A=25, B=75, Alpha=$System.Dive)
    Pan Noise=true
    Pan Noise Field=<1 0 0>
   )
   &SolveForcesAndVelocity(Clamp Velocity=false, Limit Acceleration=false)
   &SkeletalMeshLocation(
    Bone / Socket Index=3
    Bone Sampling Mode=::NewEnumerator4
    MaskSpawnGroup=false
    Orientation Sampling=::NewEnumerator1
    Skeletal Mesh=$User.SkeletalMesh
    Use Sampled Mesh Attribute as Ribbon ID=false
    Write To Particle Position=false
   )
   /Game/Effects/Niagara/Modules/Curves/FN_HermiteSpline(
    Direction Start:Vector3f=&Multiply_VectorByFloat(
     Vector=$Particles.ShapeLocation.ShapeTangent
     Float=&RandomRangeFloat(Minimum=50, Maximum=100, Randomness Mode=::NewEnumerator1, Random Seed=$Particles.SpawnGroup, Fixed Random Seed=true)
    )
    Position Blend:NiagaraFloat=$Particles.NormalizedAge
    Position End:Vector3f=&AddVectorToPosition(
     Position=&AddVectorToPosition(Position=$Particles.SkeletalMeshLocation.SampledPosition, Vector=$Particles.ShapeLocation001.ShapePosition)
     Vector=$Particles.Velocity
    )
    Position Start:Vector3f=&AddVectorToPosition(
     Position=&AddVectorToPosition(Position=$Particles.SkeletalMeshLocation.SampledPosition, Vector=$Particles.ShapeLocation.ShapePosition)
     Vector=&Multiply_VectorByFloat(Vector=$Particles.Velocity, Float=0.25)
    )
    Write Particles Position=true
   )
   /SetVariables(
    Particles.RibbonID=/Niagara/DynamicInputs/Meshes/MakeRibbonIDFromSpawnGroup()
    Particles.RibbonLinkOrder:NiagaraFloat=$Particles.NormalizedAge
   ) {isSetParameters=true, module=SetVariables_#6}
   &ScaleRibbonWidth(Ribbon Width Scale=&Waveform("Waveform [1]"=::NewEnumerator12, "[1] Frequency"=1, "[1] Amplitude Min/Max"=<1 1.5>))
   &ScaleColor(
    Scale Alpha=&FloatFromCurve(FloatCurve=DataInterface() {propertyValues {Curve {keys [
         {interpMode=RCIM_Cubic, tangentMode=RCTM_User, time=0, value=0}
         {arriveTangent=4.8895, interpMode=RCIM_Cubic, leaveTangent=-0.3407, tangentMode=RCTM_Break, time=0.5, value=1}
         <at 1 0 RCIM_Cubic>
        ]}}})
    Scale RGB=&Multiply_VectorByFloat(
     Vector=&MakeVectorFromLinearColorRGB(Color=&ColorFromCurve(ColorCurve=DataInterface() {propertyValues {
         AlphaCurve {keys [<at 0 1 RCIM_Cubic>]}
         BlueCurve {keys [<at 0.1665 2 RCIM_Cubic>, {interpMode=RCIM_Cubic, leaveTangent=-3.4472, time=0.3792, value=1}, <at 0.5927 0.0391>]}
         GreenCurve {keys [<at 0.1665 1.2823 RCIM_Cubic>, {interpMode=RCIM_Cubic, leaveTangent=-2.2542, time=0.3792, value=0.1327}, <at 0.5927 0>]}
         RedCurve {keys [<at 0.1665 0.2316 RCIM_Cubic>, {interpMode=RCIM_Cubic, leaveTangent=1.3509, time=0.3792, value=0.8716}, <at 0.5927 1>]}
        }}))
     Float=2
    )
    ScaleA=true
    ScaleRGB=true
   )
  ]
  rendererClasses [&NiagaraRibbonRendererProperties]
  renderers [{rendererClass=&NiagaraRibbonRendererProperties}]
 }
 Tendril_Ribbon_R {
  emitterUpdateScript [
   &EmitterState()
   &SpawnRate(SpawnRate=80, Use Spawn Probability=false)
  ]
  particleSpawnScript [
   &FN_SetSpawnGroup(Number of Spawn Groups=8)
   &InitializeParticle(
    Color=<rgba 1 1 1 0.2>
    Ribbon Width=1.5
    Ribbon Width Mode=::NewEnumerator1
    Sprite Size Mode=::NewEnumerator3
    Uniform Sprite Size=4
    UsePositionOffset=false
    Write Material Random=true
   )
   /SetVariables(Particles.FloatSpawnGroup:NiagaraFloat=&Divide_Float(A=&MakeFloatFromInt(Int=$Particles.SpawnGroup), B=8)) {isSetParameters=true, module=SetVariables_#5}
   &ShapeLocation(
    Apply To Particle Position=::NewEnumerator1
    MaskSpawnGroup=false
    Offset=<0 0 -15>
    Offset Coordinate Space=::World
    Offset Mode=::NewEnumerator0
    Override Seed=false
    Ring / Disc Distribution Mode:ENiagaraSphereDistributionMode=::NewEnumerator1
    Ring Radius=8
    Rotation Coordinate Space=::Simulation
    Shape Primitive=::NewEnumerator4
    U Position=$Particles.FloatSpawnGroup
    UseOwnerScale=true
   )
   &ShapeLocation(
    Apply To Particle Position=::NewEnumerator1
    MaskSpawnGroup=false
    Offset=&RandomRangeVector(
     Minimum=<0 0 50>
     Maximum=<0 0 120>
     Randomness Mode=::NewEnumerator1
     Random Seed=&Add_Integer(A=$Particles.SpawnGroup, B=8675309)
     Fixed Random Seed=true
    )
    Offset Coordinate Space=::World
    Offset Mode=::NewEnumerator0
    Override Seed=false
    Ring / Disc Distribution Mode:ENiagaraSphereDistributionMode=::NewEnumerator1
    Ring Radius=8
    Rotation Coordinate Space=::Simulation
    Shape Primitive=::NewEnumerator4
    U Position=$Particles.FloatSpawnGroup
    UseOwnerScale=true
   ) {module=ShapeLocation001}
  ]
  particleUpdateScript [
   &ParticleState()
   &InheritVelocity(Inherited Velocity Amount Scale=<0.1 0.1 -0.5>, Inherited Velocity Speed Limit=2, Speed Limit=true, UseSpeedThreshold=false)
   &CurlNoiseForce(
    Noise Frequency=&Lerp_Float(A=20, B=50, Alpha=$System.Dive)
    Noise Strength=&Lerp_Float(A=25, B=75, Alpha=$System.Dive)
    Pan Noise=true
    Pan Noise Field=<1 0 0>
   )
   &SolveForcesAndVelocity(Clamp Velocity=false, Limit Acceleration=false)
   &SkeletalMeshLocation(
    Bone / Socket Index=4
    Bone Sampling Mode=::NewEnumerator4
    MaskSpawnGroup=false
    Orientation Sampling=::NewEnumerator1
    Skeletal Mesh=$User.SkeletalMesh
    Use Sampled Mesh Attribute as Ribbon ID=false
    Write To Particle Position=false
   )
   &FN_HermiteSpline(
    Direction Start:Vector3f=&Multiply_VectorByFloat(
     Vector=$Particles.ShapeLocation.ShapeTangent
     Float=&RandomRangeFloat(
      Minimum=50
      Maximum=100
      Randomness Mode=::NewEnumerator1
      Random Seed=&Add_Integer(A=$Particles.SpawnGroup, B=8675309)
      Fixed Random Seed=true
     )
    )
    Position Blend:NiagaraFloat=$Particles.NormalizedAge
    Position End:Vector3f=&AddVectorToPosition(
     Position=&AddVectorToPosition(Position=$Particles.SkeletalMeshLocation.SampledPosition, Vector=$Particles.ShapeLocation001.ShapePosition)
     Vector=$Particles.Velocity
    )
    Position Start:Vector3f=&AddVectorToPosition(
     Position=&AddVectorToPosition(Position=$Particles.SkeletalMeshLocation.SampledPosition, Vector=$Particles.ShapeLocation.ShapePosition)
     Vector=&Multiply_VectorByFloat(Vector=$Particles.Velocity, Float=0.25)
    )
    Write Particles Position=true
   )
   /SetVariables(
    Particles.RibbonID=&MakeRibbonIDFromSpawnGroup()
    Particles.RibbonLinkOrder:NiagaraFloat=$Particles.NormalizedAge
   ) {isSetParameters=true, module=SetVariables_#6}
   &ScaleRibbonWidth(Ribbon Width Scale=&Waveform("Waveform [1]"=::NewEnumerator12, "[1] Frequency"=1, "[1] Amplitude Min/Max"=<1 1.5>))
   &ScaleColor(
    Scale Alpha=&FloatFromCurve(FloatCurve=DataInterface() {propertyValues {Curve {keys [
         {interpMode=RCIM_Cubic, tangentMode=RCTM_User, time=0, value=0}
         {arriveTangent=4.8895, interpMode=RCIM_Cubic, leaveTangent=-0.3407, tangentMode=RCTM_Break, time=0.5, value=1}
         <at 1 0 RCIM_Cubic>
        ]}}})
    Scale RGB=&Multiply_VectorByFloat(
     Vector=&MakeVectorFromLinearColorRGB(Color=&ColorFromCurve(ColorCurve=DataInterface() {propertyValues {
         AlphaCurve {keys [<at 0 1 RCIM_Cubic>]}
         BlueCurve {keys [<at 0.1665 2 RCIM_Cubic>, {interpMode=RCIM_Cubic, leaveTangent=-3.4472, time=0.3792, value=1}, <at 0.5927 0.0391>]}
         GreenCurve {keys [<at 0.1665 1.2823 RCIM_Cubic>, {interpMode=RCIM_Cubic, leaveTangent=-2.2542, time=0.3792, value=0.1327}, <at 0.5927 0>]}
         RedCurve {keys [<at 0.1665 0.2316 RCIM_Cubic>, {interpMode=RCIM_Cubic, leaveTangent=1.3509, time=0.3792, value=0.8716}, <at 0.5927 1>]}
        }}))
     Float=2
    )
    ScaleA=true
    ScaleRGB=true
   )
  ]
  rendererClasses [&NiagaraRibbonRendererProperties]
  renderers [{rendererClass=&NiagaraRibbonRendererProperties}]
 }
 Trails {
  emitterUpdateScript [
   &EmitterState()
   &SpawnBurst_Instantaneous(Spawn Count=2, Use Spawn Probability=false, UseLoopCountLimit=true)
  ]
  particleSpawnScript [
   &FN_SetSpawnGroup(Number of Spawn Groups=2)
   &InitializeParticle(Color=<rgba 1 1 1 0.35>, Mesh Scale=<0.45 0.45 3>, Mesh Scale Mode=::NewEnumerator4, UsePositionOffset=false, Write Material Random=true)
   /SetVariables(Particles.RotatorSpeedRandom:NiagaraFloat=&RandomRangeFloat(Minimum=0.3, Maximum=-0.3)) {isSetParameters=true}
   &CameraOffset(Camera Offset Amount=10) {enabled=false}
   &InitialMeshOrientation(
    Mesh Orientation Mode=::NewEnumerator7
    Rotation=&RandomRangeVector(Minimum=<0 0 0>, Maximum=<0 0 1>)
    Use Rotation Vector=true
   )
  ]
  particleUpdateScript [
   &ParticleState(Kill Particles When Lifetime Has Elapsed=false, Let Infinitely Lived Particles Die When Emitter Deactivates=true)
   &InterpolateOverTime(Rate Of Change=10, Target Vector Value:NiagaraPosition=$Particles.Position, Variable Type=::NewEnumerator2)
   /SetVariables(Particles.InterpolateDynamic:Vector4f=&MakeFloat4FromVectorAndFloat("Vector (RGB)"=$Output.InterpolateOverTime.Moving Average, "Float (Alpha)"=1)) {isSetParameters=true}
   &SkeletalMeshLocation(
    Bone / Socket Index=&Add_Integer(A=$Particles.SpawnGroup, B=3)
    Bone Sampling Mode=::NewEnumerator4
    MaskSpawnGroup=false
    Orientation Sampling=::NewEnumerator1
    Sampled Position Offset=<0 0 -60>
    Sampled Position Offset Coordinate Space:ENiagaraSkelMeshTransforms=::NewEnumerator4
    Skeletal Mesh=$User.SkeletalMesh
   )
   &SolveForcesAndVelocity(Clamp Velocity=false, Limit Acceleration=false)
   &DynamicMaterialParameters(
    Index 0 Param 1=1
    Index 0 Param 3=&Multiply_Float(A=5, B=&Lerp_Float(A=5, B=10, Alpha=$System.Dive))
    Index 2 Param 1=&Lerp_Float(A=0.1, B=0.5, Alpha=&Ceil_Float(Float=$System.Dive))
    Index 2 Param 2=&Lerp_Float(A=0, B=0.5, Alpha=&Ceil_Float(Float=$System.Dive))
    Module.Param0WriteEnabled=true
    Module.Param0WriteEnabled2=true
    Module.Param1WriteEnabled=false
    Module.Param1WriteEnabled2=true
    Module.Param2WriteEnabled=true
    Module.Param2WriteEnabled2=false
    Module.Param3WriteEnabled=false
    Module.Param3WriteEnabled2=false
    Write Parameter Index 2=true
   )
   &ScaleColor(
    Scale Alpha=&MakeCustomFloatFromBool(
     Bool=$User.bIsFrontEndPreview
     True Float=&Lerp_Float(A=0.2, B=0.3, Alpha=$System.Dive)
     False Float=&Lerp_Float(A=1, B=1, Alpha=$System.Dive)
    )
    Scale RGB=&VectorFromFloat(Value=1)
    ScaleA=true
    ScaleRGB=true
   )
   &ScaleMeshSize(Scale Factor=&MakeCustomVectorFromBool(
     Bool=$User.bIsFrontEndPreview
     True Vector=&Lerp_Vector(A=<1 1 0.9>, B=<1.4 1.4 1.6>, Alpha=$System.Dive)
     False Vector=&Lerp_Vector(A=<1.1 1.1 0.6>, B=<0.85 0.85 0.8>, Alpha=$System.Dive)
    ))
   &ScaleMeshSize(Scale Factor=&Multiply_VectorByFloat(Vector=<1.1 1.1 1>, Float=1)) {module=ScaleMeshSize001}
  ]
  rendererClasses [&NiagaraMeshRendererProperties]
  renderers [{rendererClass=&NiagaraMeshRendererProperties, rendererIndex=0}]
 }
}
summary {
 systemName=NS_Contrail_RealityBloom
 emitters [
  {emitterName=Trails, rendererClasses [&NiagaraMeshRendererProperties]}
  {emitterName=Decal_AirGlow, rendererClasses [&NiagaraSpriteRendererProperties, &NiagaraLightRendererProperties]}
  {emitterName=Petals_Mesh, rendererClasses [&NiagaraMeshRendererProperties]}
  {emitterName=Flares, rendererClasses [&NiagaraSpriteRendererProperties]}
  {emitterName=Glow_Dots_GPU, simTarget=GPUComputeSim, rendererClasses [&NiagaraSpriteRendererProperties]}
  {emitterName=Tendril_Ribbon_L, rendererClasses [&NiagaraRibbonRendererProperties]}
  {emitterName=Tendril_Ribbon_R, rendererClasses [&NiagaraRibbonRendererProperties]}
  {emitterName=Glow_Dots_CPU, rendererClasses [&NiagaraSpriteRendererProperties]}
 ]
 userVariables [
  User.bIsFrontEnd {type=NiagaraBool}
  User.bIsFrontEndPreview {type=NiagaraBool}
  User.Nurtuted {type=NiagaraFloat, defaultValue=0, description=""}
  User.Opacity {type=NiagaraFloat, defaultValue=0, description=""}
  User.ParaGlideLeanAlpha {type=NiagaraFloat, defaultValue=0, description=""}
  User.PlayerSpeed {type=NiagaraFloat, defaultValue=0, description=""}
  User.RibbonIDOffset {type=NiagaraInt32}
  User.SkeletalMesh {
   type=NiagaraDataInterfaceSkeletalMesh
   defaultValue {
    struct=NiagaraExt_VariableValue_DataInterface
    value {dataInterface=&self:NiagaraDataInterfaceSkeletalMesh_0}
   }
  }
 ]
}
